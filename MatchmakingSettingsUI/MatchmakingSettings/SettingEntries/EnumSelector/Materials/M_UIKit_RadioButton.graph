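# UI material that draws a procedural circular radio button, used by the enum selector in matchmaking settings.
# An IsOn scalar supplies the animation progress.
# MF_UI_Circle builds fill, outline and resizable inner-circle masks, which serve as named reroutes for layered color and opacity blends.
# Colors come from the shared overlay backplate function plus a near-black constant, with color premultiplied by masked vertex alpha for AlphaComposite blending.
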Material /MatchmakingSettingsUI/MatchmakingSettings/SettingEntries/EnumSelector/Materials/M_UIKit_RadioButton
graph {
 properties {bAutomaticallySetUsageInEditor=false, blendMode=BLEND_AlphaComposite, materialDomain=MD_UI, usedShadingModels=Default Lit}
 nodes [
  n1 = Multiply(A=VertexColor().A, B=NamedRerouteUsage())
  n2 = MaterialFunctionCall() {materialFunction=/MatchmakingSettingsUI/MatchmakingSettings/Materials/MF_UI_MM_OverlayBackplates}
  n3 = NamedRerouteUsage()
  n4 = Constant4Vector() {constant=<rgba 0.008 0.0075 0.016 1>}
  n5 = NamedRerouteUsage()
  Multiply(
   A=@n1
   B=LinearInterpolate(
    A=LinearInterpolate(A=@n2.B Weak Opacity, B=@n2.B Weak Opacity, Alpha=@n3) {constA=0.4, constB=0.4}
    B=LinearInterpolate(A=@n2.A Opacity, B=@n4.A, Alpha=@n5)
    Alpha=NamedRerouteUsage()
   )
  )
  n6 = NamedRerouteUsage()
  n7 = NamedRerouteUsage()
  Multiply(
   A=LinearInterpolate(
    A=LinearInterpolate(A=@n2.B Weak Color, B=@n2.B Weak Color, Alpha=@n6)
    B=LinearInterpolate(A=@n2.A Color, B=@n4.RGB, Alpha=@n7) {constA=0.6}
    Alpha=NamedRerouteUsage()
   )
   B=@n1
  )
  NamedRerouteDeclaration(None=ScalarParameter() {desc=Progress, parameterName=IsOn, sliderMax=1}) {name=Progress Ratio, nodeColor=<rgba 0 0 0 1>}
  n8 = MaterialFunctionCall(
   UVs=MaterialFunctionCall() {materialFunction=&GetUserInterfaceUV}.Normalized UV
   Size=Constant() {r=1}
   OutlineThickness=Constant() {r=0.1}
   OutlineAlignment=Constant() {r=-1}
  ) {materialFunction=&MF_UI_Circle}
  NamedRerouteDeclaration(None=@n8.Fill) {name=Circle Mask, nodeColor=<rgba 0 0 0 1>}
  NamedRerouteDeclaration(None=@n8.Outline) {name=Outline Mask, nodeColor=<rgba 0 0 0 1>}
  n9 = NamedRerouteUsage()
  NamedRerouteDeclaration(None=MaterialFunctionCall(
    UVs=MaterialFunctionCall() {materialFunction=&GetUserInterfaceUV}.Normalized UV
    Size=LinearInterpolate(Alpha=@n9) {constA=0.8, constB=0.4}
   ) {materialFunction=&MF_UI_Circle}.Fill) {name=Small Circle Mask, nodeColor=<rgba 0 0 0 1>}
 ]
}
