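# Radio button indicator widget for the matchmaking settings enum selector, made of a single non-hit-testable image using the MI_UIKit_RadioButton material.
# SetToggle plays Anim_IsOn forward or in reverse, animating the material's IsOn parameter between off and on.
# SetMobileStyle overrides the image's desired size, choosing a larger one when the compact UI is in use.
# The event graph is empty, so both functions rely on outside callers.
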WidgetBlueprint /MatchmakingSettingsUI/MatchmakingSettings/SettingEntries/EnumSelector/WBP_UIKit_RadioButton
animations [Anim_IsOn {
  binding_count=1
  length_seconds=0.2
  object=&self:Anim_IsOn
  sequence {bindings [Img_Toggle {
     id=#1
     possessed_class=Image
     tracks [{
       type=MovieSceneWidgetMaterialTrack
       sections [{type=MovieSceneParameterSection, channels [IsOn {type=MovieSceneScriptingFloatChannel, keys [[0,0],[4,1]]}]}]
      }]
    }]}
 }]
graphs [
 EventGraph {links [], nodes []}
 SetMobileStyle {nodes [
   SetMobileStyle()
   > SetDesiredSizeOverride!(
    execute
    self:Image Object Reference=Get Img_Toggle() -> (Img_Toggle:Image Object Reference)
    DesiredSize:Vector 2D Structure=Select(Option 0:Vector 2D Structure=(X=36,Y=36), Option 1:Vector 2D Structure=(X=44,Y=44), Index=IsUsingCompactUI()) -> (ReturnValue:Vector 2D Structure)
   ) -> (then)
  ]}
 SetToggle {nodes [
   n1 = SetToggle!() -> (then, On:bool=~) {is=title K2Node_FunctionEntry}
   > n2 = Branch(Condition=@n1)
   n3 = Get Anim_IsOn() -> (Anim_IsOn:Widget Animation Object Reference)
   > PlayAnimationForward(InAnimation=@n3)
   PlayAnimationReverse(execute=@n2.else, InAnimation=@n3)
  ]}
]
variables []
widgets {
 info {inheritedWidgetCount=0, parentClass=&UserWidget, rootWidgetClass=&Image, widgetCount=1}
 widgets [Img_Toggle {
   widgetClassPath=&Image
   bIsVariable=true
   widget=Img_Toggle
   properties {
    visibility=SelfHitTestInvisible
    brush {
     bIsDynamicallyLoaded=false
     drawAs=Image
     imageType=FullColor
     mirroring=NoMirror
     resourceName=None
     resourceObject=/MatchmakingSettingsUI/MatchmakingSettings/SettingEntries/EnumSelector/MaterialInstances/MI_UIKit_RadioButton
     tiling=NoTile
     imageSize=<36 36>
     margin=<ltrb 0 0 0 0>
     outlineSettings {
      bUseBrushTransparency=false
      roundingType=HalfHeightRadius
      width=0
      color {colorUseRule=UseColor_Specified, specifiedColor=<rgba 0 0 0 0>}
      cornerRadii=<0 0 0 0>
     }
     tintColor {colorUseRule=UseColor_Specified, specifiedColor=<rgba 1 1 1 1>}
     uVRegion {bIsValid=false, max=<0 0>, min=<0 0>}
    }
   }
  }]
}
