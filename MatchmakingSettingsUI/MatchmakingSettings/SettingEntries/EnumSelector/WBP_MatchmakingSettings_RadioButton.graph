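# Radio-button entry widget for matchmaking settings enum selectors, derived from WBP_MatchmakingSettings_EntryBase and adding a UIKit radio button.
# When assigned a list item, it casts the item to UIKitButtonViewModel and stores it as its view model.
# Selection and deselection call the parent handlers, then toggle the radio button on or off.
# PreConstruct replaces the inherited B_Setting border brush with a no-draw brush, hiding that background.
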WidgetBlueprint /MatchmakingSettingsUI/MatchmakingSettings/SettingEntries/EnumSelector/WBP_MatchmakingSettings_RadioButton
graphs [EventGraph {nodes [
   Event PreConstruct()
   > SetBrush!(
    execute
    self:Border Object Reference=Get B_Setting(None=Self-Reference()) -> (B_Setting:Common Border Object Reference)
    InBrush_TintColor:Slate Color Structure=(SpecifiedColor=(R=1,G=1,B=1,A=1),ColorUseRule=UseColor_Specified)
    InBrush_DrawAs:ESlateBrushDrawType Enum=NoDrawType
    InBrush_Tiling:ESlateBrushTileType Enum=NoTile
    InBrush_Mirroring:ESlateBrushMirrorType Enum=NoMirror
    InBrush_ImageSize:Vector 2D Structure=(X=32,Y=32)
    InBrush_Margin:Margin Structure=(Left=0,Top=0,Right=0,Bottom=0)
    InBrush_ResourceObject:Object Reference
    InBrush_OutlineSettings:Slate Brush Outline Settings Structure=(CornerRadii=(X=0,Y=0,Z=0,W=0),Color=(SpecifiedColor=(R=0,G=0,B=0,A=0),ColorUseRule=UseColor_Specified),Width=0,RoundingType=HalfHeightRadius,bUseBrushTransparency=False)
   ) -> (then)
   Event On Selected()
   > Parent: On Selected() {is=title K2Node_CallParentFunction}
   n1 = Get WBP_UIKit_RadioButton() -> (WBP_UIKit_RadioButton:WBP UIKit Radio Button Object Reference)
   > SetToggle!(execute, self:WBP UIKit Radio Button Object Reference=@n1, On:bool=true) -> (then)
   Event On Deselected()
   > Parent: On Deselected() {is=title K2Node_CallParentFunction}
   > SetToggle(self=@n1, On=false)
   n2 = Event OnListItemObjectSet()
   > n3 = Cast To UIKitButtonViewModel!(execute, Object=@n2.ListItemObject) -> (then, CastFailed, AsUI Kit Button View Model:UI Kit Button View Model Object Reference) {is=title K2Node_DynamicCast}
   > Set UIKitButtonViewModel(UIKitButtonViewModel:UI Kit Button View Model Object Reference=@n3) -> (Output_Get:UI Kit Button View Model Object Reference)
  ]}]
variables []
widgets {
 info {
  inheritedWidgetCount=0
  namedSlotCount=1
  parentClass=/MatchmakingSettingsUI/MatchmakingSettings/SettingEntries/WBP_MatchmakingSettings_EntryBase.WBP_MatchmakingSettings_EntryBase_C
  rootWidgetClass=null
  widgetCount=1
 }
 widgets [WBP_UIKit_RadioButton {
   widgetClassPath=&/WBP_UIKit_RadioButton
   bIsVariable=true
   widget=WBP_UIKit_RadioButton
  }]
}
